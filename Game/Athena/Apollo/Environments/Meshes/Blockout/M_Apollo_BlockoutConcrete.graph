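# Concrete surface material for Apollo environment blockout meshes, flagged for use with Nanite and spline meshes.
# A scalar Tiling parameter scales one set of UVs shared by the PowerPlant dark concrete diffuse, normal and mask texture samples.
# A Tint vector parameter, mid-dark grey by default, multiplies the diffuse color.
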Material /Game/Athena/Apollo/Environments/Meshes/Blockout/M_Apollo_BlockoutConcrete
graph {
 properties {bUsedWithNanite=true, bUsedWithSplineMeshes=true, pixelDepthOffsetMode=PDOM_Legacy, usedShadingModels=Default Lit}
 nodes [
  Constant3Vector() {constant=<rgba 0 0 1 0>}
  n1 = Multiply(A=TextureCoordinate(), B=ScalarParameter() {defaultValue=1, parameterName=Tiling})
  Multiply(
   A=TextureSample(UVs=@n1) {texture=/Game/Environments/Apollo/Sets/PowerPlant/Textures/T_PP_Concrete_Wall_1_Dark_D}.RGB
   B=VectorParameter() {parameterName=Tint, defaultValue=<rgba 0.2135 0.2135 0.2135 0>}.RGB
  )
  TextureSample(UVs=@n1) {samplerType=SAMPLERTYPE_Normal, texture=/Game/Environments/Apollo/Sets/PowerPlant/Textures/T_PP_Concrete_Wall_1_N}
  TextureSample(UVs=@n1) {samplerType=SAMPLERTYPE_Masks, texture=/Game/Environments/Apollo/Sets/PowerPlant/Textures/T_PP_Concrete_Wall_1_S}
 ]
}
